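; Thrown marking projectile that bounces and deploys a spawn where it lands, extending parent projectile logic.
; BeginPlay applies a scalable-float physics-sim spawn block and, cosmetically, flags the trail as enemy or friendly by local team.
; Hits fade the in-air audio and clamp bounce speed to prevent long flings.
; Spawn validation rejects sentries and reverts to pickup on disabled moving platforms or near pawns on enabled ones, otherwise using parent results.
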(event EventBeginPlay
  (bind _ns_shieldbubble_silverblazer_prj_trail (Variables|Default|GetNS_ShieldBubble_SilverBlazer_Prj_Trail))
  (|CallParentFunction)
  (Utilities|FlowControl|Sequence
    (:then_0
      (if (ScalableFloat|GetValueasBool (Variables|Default|GetHF_DisallowSpawnOnPhysicsSim))
        (Variables|Spawning|ValidityChecks|SetbDontSpawnOnPhysicsSim true)))
    (:then_1
      (Networking|CanExecuteCosmeticEvents
        (:True
          (if (Team|OnSameTeam (Actor|GetOwner) (Utilities|Array|Get(acopy) (Fortnite|GetLocalFortPlayerControllers true)))
            (Niagara|SetNiagaraVariable(Float) _ns_shieldbubble_silverblazer_prj_trail "EnemyRecon" 0.0)
            (else
              (Niagara|SetNiagaraVariable(Float) _ns_shieldbubble_silverblazer_prj_trail "EnemyRecon" 1.0))))))))

(event Collision|EventHit
  (params (MyComp :type Primitive%20Component%20Object%20Reference) (Other :type Actor%20Object%20Reference) (OtherComp :type Primitive%20Component%20Object%20Reference) (bSelfMoved :type Boolean) (HitLocation :type Vector :default "0, 0, 0") (HitNormal :type Vector :default "0, 0, 0") (NormalImpulse :type Vector :default "0, 0, 0") (Hit :type Hit%20Result%20Structure%20%28by%20ref%29))
  ;[1 1 1] Prevents in-air-loop from persisting after projectile lands
  (Audio|Components|Audio|FadeOut (Variables|Default|GetInAirLoop) 0.1 0.0 "Linear" :!class /Script/Engine.AudioComponent)
  (|CallParentFunction MyComp Other OtherComp bSelfMoved HitLocation HitNormal NormalImpulse Hit)
  ;[0.1771 0.6602 1] Clamp the velocity after the bounce to avoid flying 40m away
  (Projectile|SetVelocity (Variables|Movement|GetProjectileMovementComponent) (Math|Vector|ClampVectorSize (Variables|/Script/Engine.MovementComponent|GetVelocity (Variables|Movement|GetProjectileMovementComponent)) 0.0 1e+03)))

(fn IsHitValidForSpawn
  (params (HitOtherComp :type Primitive%20Component%20Object%20Reference) (HitOther :type Actor%20Object%20Reference) (HitResult :type Hit%20Result%20Structure))
  (bind _hf_movingplatformpawnradiuscheck (Variables|Default|GetHF_MovingPlatformPawnRadiusCheck))
  (bind _returnvalue (ScalableFloat|GetValueatLevel _hf_movingplatformpawnradiuscheck))
  (bind (_bblockinghit _binitialoverlap _time _distance _location _impactpoint) (Collision|BreakHitResult (Variables|GetHitResult)))
  (bind (_validspawn _reverttopickup) (|CallParentFunction HitOtherComp HitOther HitResult))
  (Utilities|FlowControl|Sequence
    (:then_0
      (Utilities|IsA(soft) HitOther "/Script/FortniteGame.BuildingGameplayActorSentry"
        (:True
          (return false _reverttopickup))))
    (:then_1
      (if (Utilities|IsValid (Class|FortBaseMovingPlatformComponent|FindMovingPlatformComponentfromActor HitOther true) :!class /Script/Engine.KismetSystemLibrary)
        (if (ScalableFloat|GetValueasBoolwithContext (Variables|Default|GetHF_MovingPlatformEnabled))
          ; Should we check for any pawns too close to deployment and revert back to pickup.
          (if (> _returnvalue 0)
            (bind (_outhit _returnvalue_2) (Collision|SphereTraceForObjects :Start _impactpoint :End _impactpoint :Radius _returnvalue :ObjectTypes (Utilities|Array|MakeArray "ObjectTypeQuery3") :bTraceComplex false :DrawDebugType "ForOneFrame" :bIgnoreSelf true :TraceColor "(R=1.000000,G=0.000000,B=0.000000,A=1.000000)" :TraceHitColor "(R=0.000000,G=1.000000,B=0.000000,A=1.000000)" :DrawTime 5.0))
            (return (not _returnvalue_2) _returnvalue_2))
          (else
            (return false true)))))
    (:then_2
      (return _validspawn _reverttopickup))))

(fn ConstructionScript
  (|CallParentFunction))
